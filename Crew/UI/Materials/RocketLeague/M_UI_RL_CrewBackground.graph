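# UI-domain material for the Rocket League crew background, layering a background texture with tinted vignettes.
# The texture is scaled about center, clamp-sampled and faded to black at two opposite edges by a mask built from a single UV channel.
# Nested lerps blend in a bottom-screen gradient color and a radial overlay color, using parameterized linear and offset, scalable radial masks.
# Named reroutes connect the aspect-scaling, texture and mask sections.
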Material /Crew/UI/Materials/RocketLeague/M_UI_RL_CrewBackground
graph {
 properties {
  bAutomaticallySetUsageInEditor=false
  bUsedWithStaticMesh=false
  materialDomain=MD_UI
  previewMesh=/Engine/Transient.MaterialEditorMeshComponent_0
  usedShadingModels=Default Lit
 }
 nodes [
  LinearInterpolate(
   A=VectorParameter() {group=Color, parameterName=RadialOverlay_Color, defaultValue=<rgba 0.0024 0.0273 0.2705 1>}.RGB
   B=LinearInterpolate(
    A=VectorParameter() {group=Color, parameterName=BottomScreenGrad_Color, defaultValue=<rgba 0.0027 0.0024 0.0103 1>}.RGB
    B=NamedRerouteUsage()
    Alpha=NamedRerouteUsage()
   )
   Alpha=NamedRerouteUsage()
  )
  NamedRerouteDeclaration(None=MaterialFunctionCall() {materialFunction=&MF_UI_UVScaling}.ScaleY) {name=ScaleY_1, nodeColor=<rgba 0 0 0 1>}
  n1 = NamedRerouteDeclaration(None=MaterialFunctionCall(UVs=NamedRerouteUsage(), Texture Scale=Constant2Vector() {g=1.1, r=1.775}) {materialFunction=&ScaleUVsByCenter}.UVs) {name=Texture Scale, nodeColor=<rgba 0 0 0 1>}
  n2 = TextureObjectParameter() {group=Background, parameterName=BackgroundTexture, texture=/Crew/UI/Textures/T_UI_RLs21_Background}
  n3 = ComponentMask(None=NamedRerouteUsage()) {r=true}
  n4 = OneMinus(None=@n3)
  n5 = Saturate(None=Min(A=Multiply(A=@n3) {constB=10}, B=Multiply(A=@n4) {constB=10}))
  NamedRerouteDeclaration(None=LinearInterpolate(
    B=TextureSample(UVs=@n1, Tex=@n2) {samplerSource=SSM_Clamp_WorldGroupSettings, texture=/Crew/UI/Textures/T_UI_RLs20_Background}.RGB
    Alpha=NamedRerouteDeclaration(None=@n5) {name=Texture Edge Mask, nodeColor=<rgba 0 0 0 1>}
   )) {name=Background Texture, nodeColor=<rgba 0 0 0 1>}
  NamedRerouteDeclaration(None=MaterialFunctionCall(
    GradientStart=ScalarParameter() {group=Vignettes, parameterName=BotScreenGrad_GradStart, sliderMax=1}
    GradientEnd=ScalarParameter() {defaultValue=1, group=Vignettes, parameterName=BotScreenGrad_GradEnd, sliderMax=1}
    RotationAngle=Constant() {r=0.25}
   ) {materialFunction=&MF_UI_LinearGradient}.Result) {name=Bottom Screen Grad Mask, nodeColor=<rgba 0 0 0 1>}
  n6 = ScalarParameter() {defaultValue=-0.2, group=Vignettes, parameterName=RadialMask_X, sliderMax=1, sliderMin=-1}
  n7 = ScalarParameter() {group=Vignettes, parameterName=RadialMask_Y, sliderMax=1, sliderMin=-1}
  n8 = ScalarParameter() {defaultValue=1.5, group=Vignettes, parameterName=RadialMask_ScaleX, sliderMax=3, sliderMin=1}
  n9 = ScalarParameter() {defaultValue=1, group=Vignettes, parameterName=RadialMask_ScaleY, sliderMax=3, sliderMin=1}
  n10 = MaterialFunctionCall(
   UVs=Add(A=NamedRerouteUsage(), B=AppendVector(A=@n6, B=@n7))
   Texture Scale=Multiply(
    A=ScalarParameter() {defaultValue=1, group=Vignettes, parameterName=RadialMask_Scale, sliderMax=3, sliderMin=0.1}
    B=AppendVector(A=@n8, B=@n9)
   )
  ) {materialFunction=&ScaleUVsByCenter}
  NamedRerouteDeclaration(None=MaterialFunctionCall(
    Input=MaterialFunctionCall(UV=@n10.UVs) {materialFunction=&MF_UI_RadialGradient}.Smooth
    Input Low=Constant() {r=-0.1}
    Input High=Constant() {r=1}
   ) {materialFunction=&RemapValue_0-1}.Result) {name=Radial Overlay Mask, nodeColor=<rgba 0 0 0 1>}
 ]
}
